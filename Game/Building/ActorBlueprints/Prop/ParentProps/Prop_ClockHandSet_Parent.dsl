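; Parent Blueprint for clock props that poses hour and minute hands to match the in-game time of day.
; On BeginPlay, non-server instances call SetTime once (despite a comment saying server), with no ongoing update.
; SetTime converts time of day into relative rotations, combines each with its hand offset and applies them to valid hands.
; GetHourHand and GetMinHand are empty here, intended for child Blueprints to supply the components.
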(event EventBeginPlay
  ;[1 1 1] Begin Play - Set the time for the min and hour hands on the server
  (|CallParentFunction)
  (if (not (Networking|IsServer))
    (CallFunction|SetTime)))

(fn GetHourHand
  (return 0))

(fn GetMinHand
  (return 0))

(fn SetTime
  (bind _hand (CallFunction|GetHourHand))
  (bind _hand_2 (CallFunction|GetMinHand))
  ;[1 1 1] Calc the rotation of the clock hands and then set the relative rotations based on time of day. Note "GetMinHand" and "GetHourHand" Should be set to the proper components in the children
  (Utilities|IsValid _hand
    (:Is%20Valid
      (Transformation|SetRelativeRotation _hand (Math|Rotator|CombineRotators (Math|Rotator|MakeRotator 0.0 (* (/ (TimeOfDay|GetTimeOfDay :!class /Script/FortniteGame.FortKismetLibrary) 12) -360)) (Variables|ClockParent|GetHourHandOffset)) false false))
    (:Is%20Not%20Valid
      _))
  (Utilities|IsValid _hand_2
    (:Is%20Valid
      (Transformation|SetRelativeRotation _hand_2 (Math|Rotator|CombineRotators (Math|Rotator|MakeRotator 0.0 (* (Math|Float|Division(WholeandRemainder) (TimeOfDay|GetTimeOfDay :!class /Script/FortniteGame.FortKismetLibrary) 1) -360)) (Variables|ClockParent|GetMinHandOffset)) false false))))

(fn ConstructionScript)
