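; Parent building prop Blueprint that drops loot when it is destroyed.
; On server-side death it calls the parent handler, then spawns loot from its configured loot tier group through the Athena function library.
; The drop point is the actor location plus a configurable relative offset, and the spawn is tagged as Destruction.
; BeginPlay, Tick, overlap and construction script hold no logic.
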(event EventBeginPlay)

(event EventOnDeathServer
  (params (Damage :type Float%20%28single-precision%29) (DamageTags :type Gameplay%20Tag%20Container%20Structure%20%28by%20ref%29) (Momentum :type Vector :default "0, 0, 0") (HitInfo :type Hit%20Result%20Structure%20%28by%20ref%29) (InstigatedBy :type Controller%20Object%20Reference) (DamageCauser :type Actor%20Object%20Reference) (EffectContext :type Gameplay%20Effect%20Context%20Handle%20Structure))
  ;[1 1 1] On Death Sever Drop Items from the specified loot tier group
  (|CallParentFunction Damage DamageTags Momentum HitInfo InstigatedBy DamageCauser EffectContext)
  (Class|AthenaFunctionLibrary|SpawnLootfromTable (Variables|DropItemsonDeath|GetLootTierGroup) 1 -1 (+ (Transformation|GetActorLocation) (Variables|DropItemsonDeath|GetLootDropRelativeLocation)) "0, 0, 0" 0 true true false 0 "Destruction"))

(event Collision|EventActorBeginOverlap
  (params (OtherActor :type Actor%20Object%20Reference)))

(event EventTick
  (params (DeltaSeconds :type Float%20%28single-precision%29)))

(fn ConstructionScript)
