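; Parent Blueprint for fire-light props that apply Curie fire interactions to actors touching their fire volume.
; On authority, the initial scan (once playlist data is ready) and overlap events begin interactions, storing handles per actor in InteractHandleMap.
; End overlap and destruction end those interactions so burning does not persist.
; It suppresses native Curie FX, fades fire audio on death, and Turn Off removes the point light and skips the initial scan.
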(event EventBeginPlay
  ;[1 1 1] Begin Play
  (|CallParentFunction)
  (Class|FortAsyncActionCurrentPlaylistReady|PlaylistDataReadyAsync
    (:OnReady
      (if (not (Variables|Default|GetTurn%20Off))
        (if (Curie|IsCurieActive)
          (CallFunction|CheckInitialOverlappingActors))))))

(event EventOnDeathPlayEffects
  (params (Damage :type Float%20%28single-precision%29) (DamageTags :type Gameplay%20Tag%20Container%20Structure%20%28by%20ref%29) (Momentum :type Vector :default "0, 0, 0") (HitInfo :type Hit%20Result%20Structure%20%28by%20ref%29) (InstigatedBy :type Fort%20Pawn%20Object%20Reference) (DamageCauser :type Actor%20Object%20Reference) (EffectContext :type Gameplay%20Effect%20Context%20Handle%20Structure))
  ;[1 1 1] On Death Play Effects
  (Audio|Components|Audio|FadeOut (Variables|Default|GetFire_Sound) 0.5 0.0 "Linear" :!class /Script/Engine.AudioComponent))

(event EventCallAnimationCurveAgain)

(event Custom|CheckInitialOverlappingActors
  ;[1 1 1] Check initial overlapping actors and handle Curie fire interactions
  (Networking|SwitchHasAuthority
    (:Authority
      (if (Curie|IsCurieActive)
        (for _array_element (Collision|GetOverlappingActors (Variables|Default|GetFire_CollisionVolume) 0 :!class /Script/Engine.PrimitiveComponent)
          (if (not (Utilities|Map|Contains (Variables|Default|GetInteractHandleMap) _array_element))
            (bind _returnvalue (Curie|ExecuteBeginFireInteract self self _array_element 1e+02 (Collision|MakeHitResult false false 0.0 0.0 (Transformation|GetActorLocation _array_element) (Transformation|GetActorLocation _array_element) "0,0,1" "0,0,1" 0 _array_element)))
            (Utilities|Map|Add (Variables|Default|GetInteractHandleMap) _array_element _returnvalue)))))))

(event OnComponentBeginOverlap(Fire_CollisionVolume)
  (params (OverlappedComponent :type Primitive%20Component%20Object%20Reference) (OtherActor :type Actor%20Object%20Reference) (OtherComp :type Primitive%20Component%20Object%20Reference) (OtherBodyIndex :type Integer) (bFromSweep :type Boolean) (SweepResult :type Hit%20Result%20Structure%20%28by%20ref%29))
  ;[1 1 1] On Fire Collision Begin Overlap
  (Networking|SwitchHasAuthority
    (:Authority
      (if (Curie|IsCurieActive)
        (if (not (Utilities|Map|Contains (Variables|Default|GetInteractHandleMap) OtherActor))
          (bind _returnvalue (Curie|ExecuteBeginFireInteract self self OtherActor 1e+02 SweepResult))
          (Utilities|Map|Add (Variables|Default|GetInteractHandleMap) OtherActor _returnvalue))))))

(event OnComponentEndOverlap(Fire_CollisionVolume)
  (params (OverlappedComponent :type Primitive%20Component%20Object%20Reference) (OtherActor :type Actor%20Object%20Reference) (OtherComp :type Primitive%20Component%20Object%20Reference) (OtherBodyIndex :type Integer))
  (bind _interacthandlemap (Variables|Default|GetInteractHandleMap))
  (bind (_value _returnvalue) (Utilities|Map|Find _interacthandlemap OtherActor))
  ;[1 1 1] On Fire Collision End Overlap
  (Networking|SwitchHasAuthority
    (:Authority
      (if (Curie|IsCurieActive)
        (if _returnvalue
          (Curie|ExecuteEndFireInteract OtherActor _value)
          (Utilities|Map|Remove (Variables|Default|GetInteractHandleMap) OtherActor))))))

(event EventDestroyed
  (bind _interacthandlemap (Variables|Default|GetInteractHandleMap))
  ;[1 1 1] Remove all fire created by this guy if destroyed.  Otherwise, people will burn eternally.
  (|CallParentFunction)
  (bind _keys (Utilities|Map|Keys (Variables|Default|GetInteractHandleMap)))
  (for _array_element _keys
    (if (.ReturnValue (Utilities|Map|Find _interacthandlemap _array_element))
      (Curie|ExecuteEndFireInteract _array_element (Utilities|Map|Find _interacthandlemap _array_element))))
  (Utilities|Map|Clear (Variables|Default|GetInteractHandleMap)))

(fn GetDynamicDestructionStaticMeshComponents
  (return (Utilities|Array|MakeArray (Variables|Debug|GetStaticMeshComponent)) (Utilities|Array|MakeArray (Variables|Default|GetBloomPlane)) (Utilities|Array|MakeArray (Variables|Debug|GetStaticMeshComponent))))

(fn ShouldPlayNativeCurieFX
  (params (FXTag :type Gameplay%20Tag%20Structure))
  (return (Utilities|Struct|MakeFortNativeCurieFXResponse :GameplayCueResponse "IgnoreCue" :bShouldPlayGeneralVFX false :bShouldPlayAmbientAudio false :bShouldPlayGlow false :bShouldPlayFXAsAOE false)))

(fn ConstructionScript
  (|CallParentFunction)
  (if (Variables|Default|GetTurn%20Off)
    (Components|DestroyComponent (Variables|Default|GetPointLight))))
